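; Player pawn for the PineSoda boss NPC, which can be hired (converted) and voices alerts, commands and vehicle entry.
; Clients poll local player distance on a timer to show the hire icon, removed when replicated HasHireIcon clears.
; Conversion strips damage-modifier effects, swaps the nav filter and binds ping and vehicle events; unconversion unbinds them and leashes the AI in place.
; The events set a replicated SoundBankTag, whose OnRep plays client-side voice.
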(event EventBeginPlay
  (|CallParentFunction)
  (Utilities|FlowControl|Sequence
    (:then_0
      (bind _asfort_athena_aibot_controller (Utilities|Casting|CastToFortAthenaAIBotController (Pawn|GetController)))
      (Variables|/Script/AIModule.AIController|SetDefaultNavigationFilterClass (Variables|Default|GetDefaultNavFilterToApplyOnSpawn) _asfort_athena_aibot_controller))
    (:then_1
      (Networking|SwitchHasAuthority
        (:Remote
          (if (Variables|Default|GetHasHireIcon)
            (bind _asaigallery_widget_can_be_hired (Utilities|Casting|CastToAIGalleryWidget_CanBeHired (UserInterface|GetUserWidgetObject (Variables|Default|GetFortWidget_HireableIcon))))
            (Variables|InternalData|SetHIredWidget _asaigallery_widget_can_be_hired)
            (bind _returnvalue (Utilities|Time|SetTimerbyEvent (EventDispatchers|CreateEvent :!event Custom|TryDisplayHireIcon) (ScalableFloat|GetValueatLevel (Variables|Default|GetCheckDrawHireIconInterval)) true))
            (Variables|InternalData|SetHireIconTimer _returnvalue)))))))

(event Custom|TryDisplayHireIcon
  (bind _array_element (Utilities|Array|ForEachLoopwithBreak (Fortnite|GetLocalFortPlayerControllers true))
    (:LoopBody
      (Utilities|IsValid (Variables|Default|GetFortWidget_HireableIcon)
        (:Is%20Valid
          (bind _hiredwidget (Variables|InternalData|GetHIredWidget)
            (:then
              (Widget|SetVisibility _hiredwidget (select (<= (Class|AbilityFunctions|GetDistanceBetweenActors (Pawn|GetControlledPawn _array_element) self) (ScalableFloat|GetValueatLevel (Variables|Default|GetCheckDistanceDrawHireIcon))) "HitTestInvisible" "Collapsed")))
            (:else)))))))

(event Custom|AIReceivedCommand
  (params (CommandType :type PingCommandType%20Enum :default "GoTo"))
  (switch Utilities|FlowControl|Switch|SwitchonPingCommandType CommandType
    (:GoTo
      _)
    (:BackToMe
      _)
    (:BackToDefault
      _)
    (:HoldPosition
      _)
    (:Revive
      (Variables|Default|SetSoundBankTag "(TagName=\"Athena.AI.Gallery.OnRevive\")")
      (Utilities|FlowControl|Delay 2.0)
      (Variables|Default|SetSoundBankTag "(TagName=\"\")"))
    (:Item
      _)
    (:Attack
      _))
  (Variables|Default|SetSoundBankTag "(TagName=\"Athena.AI.Gallery.CommandReceived\")"))

(event Custom|EnableListenEnteredVehicle
  (Default|BindEventtoOnEnteredVehicleDispatcher (delegate Custom|PawnHasEnteredVehicle)))

(event Custom|PawnHasEnteredVehicle
  (Variables|Default|SetSoundBankTag "(TagName=\"Athena.AI.Gallery.EnteredVehicle\")")
  (Utilities|FlowControl|Delay 2.0)
  (Variables|Default|SetSoundBankTag "(TagName=\"\")"))

(event Custom|DisableListenEnteredVehicle
  (Default|UnbindEventfromOnEnteredVehicleDispatcher (delegate Custom|PawnHasEnteredVehicle)))

(fn HandleAlertStateUpdates
  (CallFunction|PlayAlertLevelVO (Variables|Default|GetOldAlertLevel) (Variables|Default|GetAlertLevel)))

(fn HandlePawnConverted
  (params (InstigatorPawn :type Fort%20Pawn%20Object%20Reference) (ConvertedPawn :type Fort%20Pawn%20Object%20Reference))
  (|CallParentFunction InstigatorPawn ConvertedPawn)
  (Networking|SwitchHasAuthority
    (:Authority
      (GameplayEffects|RemoveActiveEffectswithTags (Ability|GetAbilitySystemComponent ConvertedPawn) "(GameplayTags=((TagName=\"Gameplay.Effect.NPC.Gallery.DamageModifiers\")))")
      (ConvertEvents|BindEventtoOnPingCommandEvent (Variables|FortPlayerPawn|GetConvertComponent) (EventDispatchers|CreateEvent :!event AIReceivedCommand))
      (CallFunction|EnableListenEnteredVehicle)
      (Variables|Default|SetHasHireIcon false)
      (bind _asfort_athena_aibot_controller (Utilities|Casting|CastToFortAthenaAIBotController (Pawn|GetController)))
      (Variables|/Script/AIModule.AIController|SetDefaultNavigationFilterClass (Variables|Default|GetNavFilterOnConverted) _asfort_athena_aibot_controller))))

(fn HandlePawnUnconverted
  (params (UnconvertedPawn :type Fort%20Pawn%20Object%20Reference))
  (bind _self self)
  (bind _returnvalue (Pawn|GetController _self))
  (|CallParentFunction UnconvertedPawn)
  (ConvertEvents|UnbindEventfromOnPingCommandEvent (Variables|FortPlayerPawn|GetConvertComponent) (EventDispatchers|CreateEvent :!event AIReceivedCommand))
  (CallFunction|DisableListenEnteredVehicle)
  (Utilities|IsValid _returnvalue
    (:Is%20Valid
      (bind _asfort_athena_aibot_controller (Utilities|Casting|CastToFortAthenaAIBotController _returnvalue))
      (Class|BPFLAICommonFunctions|SetRadialLeash _asfort_athena_aibot_controller (Transformation|GetActorLocation _self) (ScalableFloat|GetValueatLevel (Variables|Default|GetUnconvertedLeashInnerRadius)) (ScalableFloat|GetValueatLevel (Variables|Default|GetUnconvertedLeashOutterRadius))))))

(fn OnRep_HasHireIcon
  (if (not (Variables|Default|GetHasHireIcon))
    (Networking|SwitchHasAuthority
      (:Remote
        (Components|DestroyComponent (Variables|Default|GetFortWidget_HireableIcon))
        (Utilities|Time|ClearandInvalidateTimerbyHandle (Variables|InternalData|GetHireIconTimer))
        (Variables|InternalData|SetHIredWidget 0)))))

(fn OnRep_SoundBankTag
  (bind _soundbanktag (Variables|Default|GetSoundBankTag))
  (Networking|SwitchHasAuthority
    (:Remote
      (if (!= _soundbanktag 0)
        (CallFunction|PlayVoiceComponentSoundbyTag _soundbanktag)))))

(fn ConstructionScript
  (|CallParentFunction))
